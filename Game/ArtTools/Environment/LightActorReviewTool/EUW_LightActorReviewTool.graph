EditorUtilityWidgetBlueprint /Game/ArtTools/Environment/LightActorReviewTool/EUW_LightActorReviewTool
graphs [
 Add New Instance Data to Map {nodes [
   n1 = Add New Instance Data to Map!() -> (then, ActorStruct:"STC Light Actor Info Structure (by ref)") {is=title K2Node_FunctionEntry}
   > Set L_ActorStruct(L_ActorStruct:STC Light Actor Info Structure=@n1) -> (Output_Get:STC Light Actor Info Structure)
   > Set L_ActorClass(L_ActorClass:String=Break STC Light Actor Info!(STC_LightActorInfo:"STC Light Actor Info Structure (by ref)"=Get L_ActorStruct() -> (L_ActorStruct:STC Light Actor Info Structure)) -> (
     PathName_17_#1:String
     Class_18_#2:String
     LightClass_19_#3:String
     WorldTransform_20_#4:Transform
    ) {is=title K2Node_BreakStruct}.Class_18_#2) -> (Output_Get:String)
   n2 = Find(
    TargetMap:Map of Strings to STC Map Value Struct Structures=Get ClassMap() -> (ClassMap:Map of Strings to STC Map Value Struct Structures)
    Key=Get L_ActorClass() -> (L_ActorClass:String)
   ) -> (Value:STC Map Value Struct Structure)
   > n3 = Branch(Condition=@n2.ReturnValue)
   > n4 = Set L_InstanceList(L_InstanceList:STC Map Value Struct Structure=@n2) -> (Output_Get:STC Map Value Struct Structure)
   n5 = Get L_ActorStruct()
   > Add(
    TargetArray:Array of STC Light Actor Info Structures=Break STC Map Value Struct!(STC_MapValueStruct:"STC Map Value Struct Structure (by ref)"=@n4) -> (Instances_3_#5:Array of STC Light Actor Info Structures) {is=title K2Node_BreakStruct}
    NewItem:"STC Light Actor Info Structure (by ref)"=@n5
   ) {is=title K2Node_CallArrayFunction}
   > Add!(
    execute
    TargetMap:Map of Strings to STC Map Value Struct Structures=Get ClassMap()
    Key:"String (by ref)"=Get L_ActorClass()
    Value:"STC Map Value Struct Structure (by ref)"=Get L_InstanceList() -> (L_InstanceList:STC Map Value Struct Structure)
   ) -> (then) {is=title K2Node_CallFunction}
   n6 = Get L_ActorStruct()
   Add(
    execute=@n3.else
    TargetMap=Get ClassMap()
    Key=Get L_ActorClass()
    Value=Make STC Map Value Struct!(Instances_3_#5:Array of STC Light Actor Info Structures=Make Array("[0]":STC Light Actor Info Structure=@n6) -> (Array:Array of STC Light Actor Info Structures)) -> (STC_MapValueStruct:STC Map Value Struct Structure) {is=title K2Node_MakeStruct}
   )
   Set members in STC Light Actor Info!(execute:Exec, StructRef:"STC Light Actor Info Structure (by ref)"=@n5) -> (then:Exec, StructOut:"STC Light Actor Info Structure (by ref)") {is=title K2Node_SetFieldsInStruct}
  ]}
 DONOTUSE_GetAllActorsWithLightComponents {nodes [
   Clear(TargetArray:Array of Scene Component Object References=Get DONOTUSE_LightComponentActors() -> (DONOTUSE_LightComponentActors:Array of Scene Component Object References))
   > n1 = GetAllActorsOfClass!(execute, WorldContextObject:Object Reference, ActorClass:Actor Class Reference=&Actor) -> (then, OutActors:Array of Actor Object References)
   > n2 = For Each Loop(Array:Array of Actor Object References=@n1) -> (Array Element:Actor Object Reference)
   > n3 = For Each Loop(Array:Array of Light Component Object References=Get Components By Class!(self:Actor Object Reference=@n2, ComponentClass:Actor Component Class Reference=/Script/Engine.LightComponent) -> (ReturnValue:Array of Light Component Object References)) -> (Array Element:Light Component Object Reference)
   n4 = Get DONOTUSE_LightComponentActors()
   > Add(
    TargetArray:Array of Scene Component Object References=@n4
    NewItem:"Scene Component Object Reference (by ref)"=GetAttachParent!(self:Scene Component Object Reference=@n3) -> (ReturnValue:Scene Component Object Reference)
   ) {is=title K2Node_CallArrayFunction}
   > Return Node!(execute, FilteredActors:Array of Scene Component Object References=@n4)
   DONOTUSE_GetAllActorsWithLightComponents() {is=title K2Node_FunctionEntry}
   n5 = Get AllActors() -> (AllActors:Array of Actor Object References)
   > PrintString(
    InString=GetDisplayName(Object="Get (a copy)"(Array:Array of Actor Object References=@n5) -> (Output:Actor Object Reference))
    TextColor=(R=1,G=0.118,B=0,A=1)
   )
   n6 = For Each Loop(Array:Array of Scene Component Object References=@n4) -> (Array Element:Scene Component Object Reference)
   > PrintString(InString=GetDisplayName(Object=@n6))
   PrintString(InString=GetDisplayName(), TextColor=(R=0,G=1,B=0.3426,A=1))
  ]}
 DONOTUSE_GetAllLightActors {nodes [
   n1 = DONOTUSE_GetAllLightActors!() -> (then, Array:Array of Actor Object References) {is=title K2Node_FunctionEntry}
   > n2 = For Each Loop(Array:Array of Actor Object References=@n1) -> (Array Element:Actor Object Reference)
   > n3 = For Each Loop(Array:Array of Light Component Object References=Get Components By Class(self=@n2, ComponentClass=&LightComponent)) -> (Array Element:Light Component Object Reference)
   n4 = Get AllLightActors() -> (AllLightActors:Array of Actor Object References)
   > Add Unique(TargetArray:Array of Actor Object References=@n4, NewItem:"Actor Object Reference (by ref)"=GetOwner(self=@n3))
   > Return Node(FilteredActors:Array of Actor Object References=@n4)
  ]}
 EventGraph {nodes [
   "On Clicked (BTN_PullData)"!() -> (OutputDelegate, then, widget:User Widget Object Reference, event:Pointer Event Structure) {is=title K2Node_ComponentBoundEvent}
   > GetAllActors()
   > GetAllActorsWithLights()
   > Clear!(execute, TargetMap:Map of Strings to STC Map Value Struct Structures=Get ClassMap()) -> (then) {is=title K2Node_CallFunction}
   > n1 = For Each Loop(Array:Array of Actor Object References=Get AllLightActors()) -> (Array Element:Actor Object Reference)
   > n2 = GetActorInstanceData!(execute, self, Actor:"Actor Object Reference (by ref)"=@n1) -> (then, STC_LightActorInfo:STC Light Actor Info Structure)
   > Add New Instance Data to Map!(execute, self, ActorStruct:"STC Light Actor Info Structure (by ref)"=@n2) -> (then) {is=title K2Node_CallFunction}
   "On Clicked (BTN_PrepMap)"!() -> (OutputDelegate, then, widget:User Widget Object Reference, event:Pointer Event Structure) {is=title K2Node_ComponentBoundEvent}
   > LightMapPrep()
   "On Clicked (BTN_BuildCollection)"!() -> (OutputDelegate, then, widget:User Widget Object Reference, event:Pointer Event Structure) {is=title K2Node_ComponentBoundEvent}
   > Generate Collection from Map()
   Event Construct()
   > SetObject(self=Get DetailsView_67() -> (DetailsView_67:Details View Object Reference), NewObject=Self-Reference())
   "On Clicked (BTN_ExportData)"!() -> (OutputDelegate, then, widget:User Widget Object Reference, event:Pointer Event Structure) {is=title K2Node_ComponentBoundEvent}
  ]}
 Generate Collection from Map {nodes [
   Generate Collection from Map() {is=title K2Node_FunctionEntry}
   > n1 = Sequence()
   > n2 = Keys!(execute, TargetMap:Map of Strings to STC Map Value Struct Structures=Get ClassMap()) -> (then, Keys:Array of Strings)
   > n3 = For Each Loop(Array:Array of Strings=@n2) -> (Array Element:String)
   > Add(TargetArray:Array of Names=Get L_PathNames() -> (L_PathNames:Array of Names), NewItem:"Name (by ref)"=String To Name(InString=@n3))
   n4 = Get CollectionName() -> (CollectionName:Name=None)
   n5 = Get AssetTagsSubsystem() -> (ReturnValue:Asset Tags Subsystem Object Reference) {is=title K2Node_GetEngineSubsystem}
   n6 = CollectionExists!(execute=@n1.then_1, self:Asset Tags Subsystem Object Reference=@n5, Name:Name=@n4) -> (then, ReturnValue:bool)
   > n7 = Branch(Condition=@n6)
   > n8 = EmptyCollection!(execute, self:Asset Tags Subsystem Object Reference=@n5, Name:Name=Get CollectionName() -> (CollectionName=None)) -> (then, ReturnValue:bool)
   AddAssetsToCollection!(
    execute=[@n9.then, @n8.then]
    self:Asset Tags Subsystem Object Reference=@n5
    Name:Name=Get CollectionName() -> (CollectionName=None)
    AssetPathNames:Array of Names=Get L_PathNames()
   ) -> (then, ReturnValue:bool)
   n9 = CreateCollection!(
    execute=@n7.else
    self:Asset Tags Subsystem Object Reference=@n5
    Name:Name=@n4
    ShareType:ECollectionScriptingShareType Enum=Get CollectionType() -> (CollectionType:ECollectionScriptingShareType Enum=Local)
   ) -> (then, ReturnValue:bool)
  ]}
 GetActorDataAsStrings {nodes [
   n1 = GetActorDataAsStrings!() -> (then, Array:Array of Scene Component Object References) {is=title K2Node_FunctionEntry}
   > n2 = For Each Loop(Array:Array of Scene Component Object References=@n1) -> (Array Element:Scene Component Object Reference)
   > n3 = GetPathToLevel!(execute, self, Object:Object Reference=@n2) -> (then, ReturnValue:String)
   > Return Node!(
    execute
    Name:String=GetDisplayName(Object=@n2)
    Class:String=Get Class Display Name(Class=Get Class(Object=@n2) -> (ReturnValue:Scene Component Class Reference))
    Sublevel:String=@n3
    World Location:Vector=Get World Location!(self:Scene Component Object Reference=@n2) -> (ReturnValue:Vector) {is=title K2Node_CallFunction}
   )
  ]}
 GetActorInstanceData {nodes [
   n1 = GetActorInstanceData!() -> (then, Actor:"Actor Object Reference (by ref)") {is=title K2Node_FunctionEntry}
   > n2 = GetPackageNameFromActor!(execute, self, Actor:Object Reference=@n1) -> (then, PackageName:String)
   > Return Node!(
    execute
    STC_LightActorInfo:STC Light Actor Info Structure=Make STC Light Actor Info!(
     PathName_17_#1:String=Get Object Path String(Object=@n1)
     Class_18_#2:String=@n2
     LightClass_19_#3:String
     WorldTransform_20_#4:Transform=Get Actor Transform(self=@n1)
    ) -> (STC_LightActorInfo:STC Light Actor Info Structure) {is=title K2Node_MakeStruct}
   )
  ]}
 GetAllActors {nodes [
   GetAllActors() {is=title K2Node_FunctionEntry}
   > n1 = GetAllActorsOfClass(ActorClass=&Actor)
   > n2 = Set AllActors(AllActors:Array of Actor Object References=@n1) -> (Output_Get:Array of Actor Object References)
   > Set TotalActorCount(TotalActorCount:int=Length(TargetArray=@n2)) -> (Output_Get:int=0)
  ]}
 GetAllActorsWithLights {nodes [
   GetAllActorsWithLights() {is=title K2Node_FunctionEntry}
   > n1 = For Each Loop(Array:Array of Actor Object References=Get AllActors()) -> (Array Element:Actor Object Reference)
   > Is Valid(InputObject=GetComponentByClass!(self:Actor Object Reference=@n1, ComponentClass:Actor Component Class Reference=&LightComponent) -> (ReturnValue:Light Component Object Reference))
   > Add(TargetArray:Array of Actor Object References=Get AllLightActors(), NewItem:"Actor Object Reference (by ref)"=@n1)
  ]}
 GetPackageNameFromActor {nodes [
   n1 = GetPackageNameFromActor!() -> (then, Actor:Object Reference) {is=title K2Node_FunctionEntry}
   n2 = Get Class(Object=@n1) -> (ReturnValue:Object Class Reference)
   > n3 = Execute Python Script!(execute, PythonScript="import unreal\r\nclass_path = class_in.get_path_name()", class_in:Object Class Reference=@n2) -> (then, ReturnValue, class_path:String) {inputs [class_in], outputs [class_path]}
   n4 = Split(SourceString=@n3.class_path, InStr=_C, SearchDir=FromEnd)
   > Return Node!(
    execute
    PackageName:String=Select(Option 0:String=Get Class Display Name(Class=@n2), Option 1:String=@n4, Index=@n4.ReturnValue) -> (ReturnValue:String)
   )
  ]}
 GetPathToLevel {nodes [
   n1 = GetPathToLevel!() -> (then, Object:Object Reference) {is=title K2Node_FunctionEntry}
   > Return Node!(execute, ReturnValue:String=Get Object Path String(Object=@n1))
   n2 = GetOuterObject!(Object:Object Reference) -> (ReturnValue:Object Reference)
   GetObjectName(Object=GetOuterObject(Object=GetOuterObject(Object=@n2)))
   n3 = GetOuterObject(Object=GetOuterObject())
   Split(SourceString=GetObjectName(Object=GetOuterObject(Object=@n3)), InStr=/, SearchDir=FromEnd)
  ]}
 GetTotalNumberofLights {nodes [
   GetTotalNumberofLights() {is=title K2Node_FunctionEntry}
   > Clear(TargetArray=Get FilteredActors() -> (FilteredActors:Array of Scene Component Object References)) {is=title K2Node_CallArrayFunction}
   > Clear(TargetArray=Get L_AllLightActorComponents() -> (L_AllLightActorComponents:Array of Scene Component Object References))
   > n1 = For Each Loop(Array=Get AllActors())
   > n2 = For Each Loop(Array:Array of Light Component Object References=Get Components By Class(self=@n1, ComponentClass=&LightComponent)) -> (Array Element:Light Component Object Reference)
   n3 = Get L_AllLightActorComponents()
   Set FilteredActors(execute=@n2.Completed, FilteredActors:Array of Scene Component Object References=@n3) -> (Output_Get:Array of Scene Component Object References)
   > Set TotalLightCount(TotalLightCount:int=Length(TargetArray:Array of Scene Component Object References=@n3)) -> (Output_Get:int=0)
   Add(execute=@n2.LoopBody, TargetArray:Array of Scene Component Object References=@n3, NewItem:"Scene Component Object Reference (by ref)"=@n2)
  ]}
 LightMapPrep {nodes [
   LightMapPrep() {is=title K2Node_FunctionEntry}
   > n1 = Sequence!(execute) -> (then_0, then_1, then_2:Exec)
   > Set TotalFoundationCount(TotalFoundationCount:int=1) -> (Output_Get:int=0)
   Set IterativeFoundationCount(execute=@n1.then_1, IterativeFoundationCount:int=0) -> (Output_Get:int=0)
   PrintString(execute=@n1.then_2, InString=Actor counts reset, TextColor=(R=1,G=0.6924,B=0,A=1), Duration=3)
   > n2 = GetAllActorsOfClass(ActorClass=/Script/FortniteGame.BuildingFoundation) -> (OutActors:Array of Building Foundation Object References)
   > Set TotalFoundationCount(TotalFoundationCount=Length(TargetArray:Array of Building Foundation Object References=@n2)) -> (Output_Get=0)
   > n3 = While Loop!(
    execute:Exec
    Condition:bool="Not Equal (Integer)"(A=Get IterativeFoundationCount() -> (IterativeFoundationCount:int=0), B=Get TotalFoundationCount() -> (TotalFoundationCount:int=0))
   ) -> (LoopBody:Exec, Completed:Exec) {is=title K2Node_MacroInstance}
   PrintString(execute=@n3.Completed, InString=All actors loaded, TextColor=(R=0.1846,G=1,B=0,A=1), Duration=5)
   PreviewAllFoundations(execute=@n3.LoopBody)
   > LoopFoundationCounter()
  ]}
 LoopFoundationCounter {nodes [
   LoopFoundationCounter() {is=title K2Node_FunctionEntry}
   > n1 = GetAllActorsOfClass(ActorClass=&BuildingFoundation)
   > Set IterativeFoundationCount(IterativeFoundationCount=Length(TargetArray=@n1)) -> (Output_Get=0)
   > Set LoopFoundationCount(LoopFoundationCount:int=int + int(A=Get LoopFoundationCount() -> (LoopFoundationCount:int=0))) -> (Output_Get:int=0)
   > Branch(Condition="Equal (Integer)"(A=Get LoopFoundationCount() -> (LoopFoundationCount=0), B=3))
   > Set TotalFoundationCount(TotalFoundationCount=Get IterativeFoundationCount() -> (IterativeFoundationCount=0)) -> (Output_Get=0)
  ]}
 PreviewAllFoundations {nodes [
   PreviewAllFoundations() {is=title K2Node_FunctionEntry}
   > n1 = GetAllActorsOfClass(ActorClass=&BuildingFoundation)
   > n2 = For Each Loop(Array:Array of Building Foundation Object References=@n1) -> (Array Element:Building Foundation Object Reference)
   > SetEditorProperty(Object=@n2, PropertyName=bPreviewFoundationLevelInEditor, PropertyValue=MakeLiteralBool(Value=true))
  ]}
]
variables [
 ActorData {type="array<string>", default []}
 ActorInfo {type="array<STC_LightActorInfo>", default []}
 AllActors {type="array<Actor>"}
 AllLightActors {type="array<Actor>"}
 Array {type="array<STC_LightActorInfo>", default []}
 ClassMap {type="map<STC_MapValueStruct>", default {}}
 CollectionName {type=string, default=LightActorReview}
 CollectionType {type=ECollectionScriptingShareType, default=Local}
 DelimiterCharacter {type=string, default=","}
 DONOTUSE_LightComponentActors {type="array<SceneComponent>", default []}
 FilteredActors {type="array<SceneComponent>", default []}
 IterativeFoundationCount {type=integer}
 LoopFoundationCount {type=integer}
 PathNames {type="array<string>", default []}
 Struct Out {
  type=STC_LightActorInfo
  default {class="", light Class="", path Name="", world Transform {location=<0 0 0>, rotation=<rot 0 0 0>, scale=<1 1 1>}}
 }
 TotalActorCount {type=integer}
 TotalFoundationCount {type=integer}
 TotalLightActorCount {type=integer}
 TotalLightCount {type=integer}
]
widgets {
 info {rootWidgetClass=&VerticalBox, widgetCount=8}
 widgets [
  VerticalBox_0 {widgetClassPath=&VerticalBox, widget=VerticalBox_0}
  HorizontalBox_3 {
   widgetClassPath=&HorizontalBox
   parent=VerticalBox_0
   widget=HorizontalBox_3
   slot {class=VerticalBoxSlot, properties {size {sizeRule=Fill, value=1}}}
  }
  VB_Buttons {
   widgetClassPath=&VerticalBox
   parent=HorizontalBox_3
   widget=VB_Buttons
   slot {class=HorizontalBoxSlot, properties {horizontalAlignment=HAlign_Left, verticalAlignment=VAlign_Top, padding=<ltrb 5 5 5 5>}}
  }
  BTN_PrepMap {
   widgetClassPath=&TA_Generic01_Button
   parent=VB_Buttons
   widget=BTN_PrepMap
   properties {label=TEMP Prep Map, toolTipText=Finds all unloaded foundations and previews them until actor count no longer increases}
   slot {
    class=VerticalBoxSlot
    properties {horizontalAlignment=HAlign_Center, verticalAlignment=VAlign_Top, padding=<ltrb 5 5 5 5>, size {sizeRule=Fill, value=1}}
   }
  }
  BTN_PullData {
   widgetClassPath=&TA_Generic01_Button
   parent=VB_Buttons
   widget=BTN_PullData
   properties {
    label=Get Light Actor Info
    toolTipText=Looks through assets and identifies those with light components and gathers additional information about them
   }
   slot {
    class=VerticalBoxSlot
    properties {horizontalAlignment=HAlign_Center, verticalAlignment=VAlign_Top, padding=<ltrb 5 5 5 5>, size {sizeRule=Fill, value=1}}
   }
  }
  BTN_BuildCollection {
   widgetClassPath=&TA_Generic01_Button
   parent=VB_Buttons
   widget=BTN_BuildCollection
   properties {label=Make Collection, toolTipText=Clear debug list}
   slot {
    class=VerticalBoxSlot
    properties {horizontalAlignment=HAlign_Center, verticalAlignment=VAlign_Top, padding=<ltrb 5 5 5 5>, size {sizeRule=Fill, value=1}}
   }
  }
  BTN_ExportData {
   widgetClassPath=&TA_Generic01_Button
   parent=VB_Buttons
   widget=BTN_ExportData
   properties {label=TEMP Export Data, toolTipText=Clear debug list}
   slot {
    class=VerticalBoxSlot
    properties {horizontalAlignment=HAlign_Center, verticalAlignment=VAlign_Top, padding=<ltrb 5 5 5 5>, size {sizeRule=Fill, value=1}}
   }
  }
  DetailsView_67 {
   widgetClassPath=&DetailsView
   parent=HorizontalBox_3
   widget=DetailsView_67
   properties {propertiesToShow [ClassMap]}
   slot {class=HorizontalBoxSlot, properties {size {sizeRule=Fill, value=1}}}
  }
 ]
}
